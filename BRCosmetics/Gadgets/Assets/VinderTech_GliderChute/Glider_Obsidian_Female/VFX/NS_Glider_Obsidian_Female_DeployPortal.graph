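# Niagara system for the portal effect of the Obsidian Female glider, named for its deploy moment.
# Four oriented mesh layers (Core, Edge, Airglow, Flamelicks) share a scale curve that snaps open, holds, then collapses.
# Camera-offset sprites add a dark backing disc and an expanding chromatic flare, alongside a magenta-violet light.
# A burst of purple sparkles with curl noise, gravity and drag joins continuously spawned sphere sparks pulled by point attraction.
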NiagaraSystem /BRCosmetics/Gadgets/Assets/VinderTech_GliderChute/Glider_Obsidian_Female/VFX/NS_Glider_Obsidian_Female_DeployPortal
dependencies {usedDataInterfaces [&NiagaraDataInterfaceCurve], usedDynamicInputs [], usedModules [&SystemState]}
emitters {
 Airglow {
  emitterUpdateScript [
   &EmitterState(
    Inactive Response=::NewEnumerator1
    Life Cycle Mode=::NewEnumerator1
    Loop Behavior=::NewEnumerator1
    Loop Delay=0.1
    Loop Duration=1
    UseLoopDelay=true
   )
   &SpawnBurst_Instantaneous(Spawn Count=1, Use Spawn Probability=false)
  ]
  particleSpawnScript [
   &InitializeParticle(Color=<rgba 1 1 1 0.25>, Mesh Scale Mode=::NewEnumerator3, Mesh Uniform Scale=5, UsePositionOffset=false, Write Material Random=true)
   &InitialMeshOrientation(Orientation Axis=<0 1 0>, Use Orientation Vector=true, Use Rotation Vector=false)
  ]
  particleUpdateScript [
   &ParticleState(Kill Particles When Lifetime Has Elapsed=false, Let Infinitely Lived Particles Die When Emitter Deactivates=true)
   &ScaleMeshSize(Scale Factor=&VectorFromFloat(Value=&FloatFromCurve(FloatCurve=DataInterface() {propertyValues {Curve {keys [
          {arriveTangent=6.6826, interpMode=RCIM_Cubic, leaveTangent=6.6826, tangentMode=RCTM_User, time=0, value=0}
          {arriveTangent=1.914, interpMode=RCIM_Cubic, leaveTangent=0.736, tangentMode=RCTM_Break, time=0.0951, value=0.9887}
          {arriveTangent=-0.7631, interpMode=RCIM_Cubic, leaveTangent=-1.2047, tangentMode=RCTM_Break, time=0.864, value=0.9964}
          {arriveTangent=-5.5008, interpMode=RCIM_Cubic, leaveTangent=-5.5008, tangentMode=RCTM_User, time=1, value=0}
         ]}}})))
  ]
  rendererClasses [&NiagaraMeshRendererProperties]
  renderers [{rendererClass=&NiagaraMeshRendererProperties, rendererIndex=0}]
 }
 Black_Back {
  emitterUpdateScript [
   &EmitterState(Inactive Response=::NewEnumerator1, Life Cycle Mode=::NewEnumerator1, Loop Behavior=::NewEnumerator1, Loop Duration=1, UseLoopDelay=true)
   &SpawnBurst_Instantaneous(Spawn Count=1, Spawn Time=0.05, Use Spawn Probability=false)
  ]
  particleSpawnScript [
   /Niagara/Modules/Spawn/Initialization/InitializeParticle(
    Color=<rgba 0 0 0 0.75>
    Lifetime=0.5
    Module.Write Color=true
    Module.Write Lifetime=true
    Module.Write Mass=true
    Module.Write Material Random=true
    Module.Write Position=true
    Module.Write Scale=false
    Module.Write SpriteRotation=true
    Module.Write SpriteSize=true
    Sprite Rotation=&RandomRangeFloat(Minimum=0, Maximum=360)
    Sprite Size=&Vector2DFromFloat(Value=1300)
   )
   &CameraOffset(Camera Offset Amount=-50)
  ]
  particleUpdateScript [
   &ParticleState()
   &ScaleColor(
    Scale Alpha=&FloatFromCurve(FloatCurve=DataInterface() {propertyValues {Curve {keys [
         {arriveTangent=5.4315, interpMode=RCIM_Cubic, leaveTangent=5.4315, tangentMode=RCTM_User, time=0, value=0}
         {arriveTangent=1.1111, leaveTangent=1.1111, time=0.25, value=1}
         {arriveTangent=-1.3333, interpMode=RCIM_Cubic, leaveTangent=-1.3333, time=0.6, value=1}
         <at 1 0 RCIM_Cubic>
        ]}}})
    ScaleA=true
    ScaleRGB=false
   )
  ]
  rendererClasses [&NiagaraSpriteRendererProperties]
  renderers [{rendererClass=&NiagaraSpriteRendererProperties}]
 }
 Chromatic_Flare {
  emitterUpdateScript [
   &EmitterState(Inactive Response=::NewEnumerator1, Life Cycle Mode=::NewEnumerator1, Loop Behavior=::NewEnumerator1, Loop Duration=1, UseLoopDelay=true)
   &SpawnBurst_Instantaneous(Spawn Count=1, Spawn Time=0.05, Use Spawn Probability=false)
  ]
  particleSpawnScript [
   &InitializeParticle(
    Color=<rgba 2 2 2 0.25>
    Lifetime=0.9
    Module.Write Color=true
    Module.Write Lifetime=true
    Module.Write Mass=true
    Module.Write Material Random=true
    Module.Write Position=true
    Module.Write Scale=false
    Module.Write SpriteRotation=true
    Module.Write SpriteSize=true
    Sprite Rotation=&RandomRangeFloat(Minimum=0, Maximum=360)
    Sprite Size=&Vector2DFromFloat(Value=1300)
   )
   &CameraOffset(Camera Offset Amount=400)
  ]
  particleUpdateScript [
   &ParticleState()
   &ScaleSpriteSize(Scale Factor=&Vector2DFromFloat(Value=&FloatFromCurve(FloatCurve=DataInterface() {propertyValues {Curve {keys [{arriveTangent=5.3461, interpMode=RCIM_Cubic, leaveTangent=5.3461, tangentMode=RCTM_User, time=0, value=0}, <at 1 2 RCIM_Cubic>]}}})))
   &DynamicMaterialParameters(
    Index 0 Param 1=0.5
    Index 0 Param 2=1
    Module.Param0WriteEnabled=true
    Module.Param1WriteEnabled=true
    Module.Param2WriteEnabled=true
    Module.Param3WriteEnabled=true
   )
   &ScaleColor(
    Scale Alpha=&FloatFromCurve(FloatCurve=DataInterface() {propertyValues {Curve {keys [<at 0.9 1 RCIM_Cubic>,<at 1 0 RCIM_Cubic>]}}})
    Scale RGB=&VectorFromFloat(Value=&FloatFromCurve(FloatCurve=DataInterface() {propertyValues {Curve {keys [
          <at 0 0 RCIM_Cubic>
          {arriveTangent=1.2835, interpMode=RCIM_Cubic, leaveTangent=1.2835, time=0.1, value=1}
          {arriveTangent=-0.937, interpMode=RCIM_Cubic, leaveTangent=-0.937, tangentMode=RCTM_User, time=0.331, value=0.4248}
          <at 1 0 RCIM_Cubic>
         ]}}}))
    ScaleA=true
    ScaleRGB=true
   )
  ]
  rendererClasses [&NiagaraSpriteRendererProperties]
  renderers [{rendererClass=&NiagaraSpriteRendererProperties}]
 }
 Core {
  emitterUpdateScript [
   &EmitterState(
    Inactive Response=::NewEnumerator1
    Life Cycle Mode=::NewEnumerator1
    Loop Behavior=::NewEnumerator1
    Loop Delay=0.1
    Loop Duration=1
    UseLoopDelay=true
   ) {module=EmitterState001}
   &SpawnBurst_Instantaneous(Spawn Count=1, Use Spawn Probability=false)
  ]
  particleSpawnScript [
   /Niagara/Modules/Spawn/Initialization/V2/InitializeParticle(Color Mode=::NewEnumerator0, Mesh Scale Mode=::NewEnumerator3, Mesh Uniform Scale=5, UsePositionOffset=false, Write Material Random=true)
   &InitialMeshOrientation(Orientation Axis=<0 1 0>, Use Orientation Vector=true, Use Rotation Vector=false)
  ]
  particleUpdateScript [
   &ParticleState(Kill Particles When Lifetime Has Elapsed=false, Let Infinitely Lived Particles Die When Emitter Deactivates=true)
   &ScaleMeshSize(Scale Factor=&VectorFromFloat(Value=&FloatFromCurve(FloatCurve=DataInterface() {propertyValues {Curve {keys [
          {arriveTangent=6.6826, interpMode=RCIM_Cubic, leaveTangent=6.6826, tangentMode=RCTM_User, time=0, value=0}
          {arriveTangent=1.914, interpMode=RCIM_Cubic, leaveTangent=0.736, tangentMode=RCTM_Break, time=0.0951, value=0.9887}
          {arriveTangent=-0.7631, interpMode=RCIM_Cubic, leaveTangent=-1.2047, tangentMode=RCTM_Break, time=0.864, value=0.9964}
          {arriveTangent=-5.5008, interpMode=RCIM_Cubic, leaveTangent=-5.5008, tangentMode=RCTM_User, time=1, value=0}
         ]}}})))
  ]
  rendererClasses [&NiagaraMeshRendererProperties]
  renderers [{rendererClass=&NiagaraMeshRendererProperties, rendererIndex=0}]
 }
 Edge {
  emitterUpdateScript [
   &EmitterState(
    Inactive Response=::NewEnumerator1
    Life Cycle Mode=::NewEnumerator1
    Loop Behavior=::NewEnumerator1
    Loop Delay=0.1
    Loop Duration=1
    UseLoopDelay=true
   ) {module=EmitterState001}
   &SpawnBurst_Instantaneous(Spawn Count=1, Use Spawn Probability=false)
  ]
  particleSpawnScript [
   &InitializeParticle(Color Mode=::NewEnumerator0, Mesh Scale Mode=::NewEnumerator3, Mesh Uniform Scale=5, UsePositionOffset=false, Write Material Random=true)
   &InitialMeshOrientation(Orientation Axis=<0 1 0>, Use Orientation Vector=true, Use Rotation Vector=false)
  ]
  particleUpdateScript [
   &ParticleState(Kill Particles When Lifetime Has Elapsed=false, Let Infinitely Lived Particles Die When Emitter Deactivates=true)
   &ScaleMeshSize(Scale Factor=&VectorFromFloat(Value=&FloatFromCurve(FloatCurve=DataInterface() {propertyValues {Curve {keys [
          {arriveTangent=6.6826, interpMode=RCIM_Cubic, leaveTangent=6.6826, tangentMode=RCTM_User, time=0, value=0}
          {arriveTangent=1.914, interpMode=RCIM_Cubic, leaveTangent=0.736, tangentMode=RCTM_Break, time=0.0951, value=0.9887}
          {arriveTangent=-0.7631, interpMode=RCIM_Cubic, leaveTangent=-1.2047, tangentMode=RCTM_Break, time=0.864, value=0.9964}
          {arriveTangent=-5.5008, interpMode=RCIM_Cubic, leaveTangent=-5.5008, tangentMode=RCTM_User, time=1, value=0}
         ]}}})))
  ]
  rendererClasses [&NiagaraMeshRendererProperties]
  renderers [{rendererClass=&NiagaraMeshRendererProperties, rendererIndex=0}]
 }
 Flamelicks {
  emitterUpdateScript [
   &EmitterState(
    Inactive Response=::NewEnumerator1
    Life Cycle Mode=::NewEnumerator1
    Loop Behavior=::NewEnumerator1
    Loop Delay=0.1
    Loop Duration=1
    UseLoopDelay=true
   ) {module=EmitterState001}
   &SpawnBurst_Instantaneous(Spawn Count=1, Use Spawn Probability=false)
  ]
  particleSpawnScript [
   &InitializeParticle(Color Mode=::NewEnumerator0, Mesh Scale Mode=::NewEnumerator3, Mesh Uniform Scale=5, UsePositionOffset=false, Write Material Random=true)
   &InitialMeshOrientation(Orientation Axis=<0 1 0>, Use Orientation Vector=true, Use Rotation Vector=false)
  ]
  particleUpdateScript [
   &ParticleState(Kill Particles When Lifetime Has Elapsed=false, Let Infinitely Lived Particles Die When Emitter Deactivates=true)
   &ScaleMeshSize(Scale Factor=&VectorFromFloat(Value=&FloatFromCurve(FloatCurve=DataInterface() {propertyValues {Curve {keys [
          {arriveTangent=6.6826, interpMode=RCIM_Cubic, leaveTangent=6.6826, tangentMode=RCTM_User, time=0, value=0}
          {arriveTangent=1.914, interpMode=RCIM_Cubic, leaveTangent=0.736, tangentMode=RCTM_Break, time=0.0951, value=0.9887}
          {arriveTangent=-0.7631, interpMode=RCIM_Cubic, leaveTangent=-1.2047, tangentMode=RCTM_Break, time=0.864, value=0.9964}
          {arriveTangent=-5.5008, interpMode=RCIM_Cubic, leaveTangent=-5.5008, tangentMode=RCTM_User, time=1, value=0}
         ]}}})))
  ]
  rendererClasses [&NiagaraMeshRendererProperties]
  renderers [{rendererClass=&NiagaraMeshRendererProperties, rendererIndex=0}]
 }
 Light001 {
  emitterUpdateScript [
   &EmitterState()
   &SpawnBurst_Instantaneous(Spawn Count=1, Spawn Time=0.2, Use Spawn Probability=false)
  ]
  particleSpawnScript [
   /Niagara/Modules/Spawn/Initialization/InitializeParticle(
    Color=<rgba 13.8474 1.4526 20 1>
    Module.Write Color=true
    Module.Write Lifetime=true
    Module.Write Mass=true
    Module.Write Material Random=true
    Module.Write Position=false
    Module.Write Scale=false
    Module.Write SpriteRotation=false
    Module.Write SpriteSize=true
    Position=&SimulationPosition()
    Sprite Size=&Vector2DFromFloat(Value=150)
   )
   &DynamicMaterialParameters(
    Index 0 Param 1=1
    Module.Param0WriteEnabled=true
    Module.Param1WriteEnabled=true
    Module.Param2WriteEnabled=true
    Module.Param3WriteEnabled=true
   )
  ]
  particleUpdateScript [
   &CameraOffset(Camera Offset Amount=15)
   &ParticleState()
   &ScaleColor(Scale Alpha=0.1, ScaleA=true, ScaleRGB=true)
   &FN_Light_v2(Light Exponent=3, Light Radius=60, LightEnabled=true, Position Offset:Vector3f=<5 -5 -5>, Scale Color=100)
  ]
  rendererClasses [&NiagaraLightRendererProperties]
  renderers [{rendererClass=&NiagaraLightRendererProperties, rendererIndex=0}]
 }
 Magic_Sparklies_Burst {
  emitterUpdateScript [
   &EmitterState(Life Cycle Mode=::NewEnumerator1, Loop Behavior=::NewEnumerator1, Loop Delay=0.2, Loop Duration=0.3, UseLoopDelay=true)
   &SpawnBurst_Instantaneous(Spawn Count=55, Use Spawn Probability=false)
  ]
  particleSpawnScript [
   &InitializeParticle(
    Color=&RandomRangeLinearColor(Minimum=<rgba 0.6467 0.529 1 1>, Maximum=<rgba 0.5862 0.045 1 1>)
    Lifetime=&RandomRangeFloat(Minimum=0.25, Maximum=0.5)
    Module.Write Color=true
    Module.Write Lifetime=true
    Module.Write Mass=true
    Module.Write Material Random=true
    Module.Write Position=true
    Module.Write Scale=false
    Module.Write SpriteRotation=false
    Module.Write SpriteSize=true
    Sprite Size=&Vector2DFromFloat(Value=&RandomRangeFloat(Minimum=30, Maximum=40))
   )
   &CameraOffset(Camera Offset Amount=50)
   &SphereLocation(
    MaskSpawnGroup=false
    Override Seed=false
    Surface Expansion Mode=::NewEnumerator1
    Surface Only=true
    UseNonUniformScale=false
    UseOffset=false
    UseSphereOrientation=false
   )
   &DynamicMaterialParameters(
    Index 0 Param 1=0.5
    Index 0 Param 2=1
    Index 0 Param 3=1
    Module.Param0WriteEnabled=true
    Module.Param1WriteEnabled=true
    Module.Param2WriteEnabled=true
    Module.Param3WriteEnabled=false
   )
   &AddVelocityFromPoint(Module.Use Falloff Distance=false, Velocity Strength=&RandomRangeFloat(Minimum=600, Maximum=1000))
  ]
  particleUpdateScript [
   &ParticleState()
   &ScaleColor(
    Scale Alpha=&FloatFromCurve(FloatCurve=DataInterface() {propertyValues {Curve {keys [<at 0.125 1 RCIM_Cubic>,<at 1 0 RCIM_Cubic>]}}})
    ScaleA=true
    ScaleRGB=true
   )
   &CurlNoiseForce(Noise Frequency=1, Noise Strength=150, Pan Noise=true)
   &AccelerationForce(Acceleration=<0 0 -250>)
   &Drag(Drag=&RandomRangeFloat(Minimum=4, Maximum=6), Use Linear Drag=true, Use Rotational Drag=false)
   &SolveForcesAndVelocity(Clamp Velocity=false, Limit Acceleration=false)
  ]
  rendererClasses [&NiagaraSpriteRendererProperties]
  renderers [{rendererClass=&NiagaraSpriteRendererProperties}]
 }
 Particle_Spark_Sphere001 {
  emitterUpdateScript [
   &EmitterState()
   &SpawnRate(SpawnRate=40, Use Spawn Probability=false)
  ]
  particleSpawnScript [
   &InitializeParticle(
    Color=&MakeCustomLinearColorFromBool(Bool=&RandomBool(), True Color=<rgba 0.9823 0.4911 0.9004 1>, False Color=<rgba 6 3 5.5 1>)
    Lifetime=&UniformRangedFloat(Minimum=0.3, Maximum=0.45)
    Module.Write Color=true
    Module.Write Lifetime=true
    Module.Write Mass=true
    Module.Write Material Random=true
    Module.Write Position=false
    Module.Write Scale=false
    Module.Write SpriteRotation=true
    Module.Write SpriteSize=true
    Position=&SimulationPosition()
    Sprite Rotation=&UniformRangedFloat(Minimum=180, Maximum=-180)
    Sprite Size=&Vector2DFromFloat(Value=&UniformRangedFloat(Minimum=6, Maximum=9))
   )
   &SphereLocation(
    MaskSpawnGroup=false
    Non Uniform Scale=<1 0 1>
    Override Seed=false
    Sphere Radius=80
    Surface Expansion Mode=::NewEnumerator1
    Surface Only=true
    Surface Only Band Thickness=0.1
    UseNonUniformScale=false
    UseOffset=true
    UseSphereOrientation=false
   )
   &CameraOffset(Camera Offset Amount=-10)
   &AddVelocityFromPoint(Module.Use Falloff Distance=false, Velocity Strength=700)
  ]
  particleUpdateScript [
   &ParticleState()
   &ScaleColor(
    Scale Alpha=&FloatFromCurve(FloatCurve=DataInterface() {propertyValues {Curve {keys [
         {interpMode=RCIM_Cubic, tangentMode=RCTM_Break, time=0, value=0}
         <at 0.2 1>
         {interpMode=RCIM_Cubic, tangentMode=RCTM_User, time=0.8, value=1}
         {interpMode=RCIM_Cubic, tangentMode=RCTM_Break, time=1, value=0}
        ]}}})
    Scale RGB=&VectorFromFloat(Value=&UniformRangedFloat(Minimum=40, Maximum=150))
    ScaleA=true
    ScaleRGB=true
   )
   &Drag(Drag=2, Use Linear Drag=true, Use Rotational Drag=false)
   &CurlNoiseForce(Noise Frequency=100, Noise Strength=25, Pan Noise=false)
   &PointAttractionForce(
    Attraction Radius=200
    Attraction Strength=40
    Falloff Exponent=0.8
    Kill Within Radius=false
    Module.Use Falloff=true
    UseAttractorOffset=false
   )
   &SolveForcesAndVelocity(Clamp Velocity=false, Limit Acceleration=false)
  ]
  rendererClasses [&NiagaraSpriteRendererProperties]
  renderers [{rendererClass=&NiagaraSpriteRendererProperties}]
 }
}
summary {
 systemName=NS_Glider_Obsidian_Female_DeployPortal
 emitters [
  {emitterName=Airglow, rendererClasses [&NiagaraMeshRendererProperties]}
  {emitterName=Core, rendererClasses [&NiagaraMeshRendererProperties]}
  {emitterName=Edge, rendererClasses [&NiagaraMeshRendererProperties]}
  {emitterName=Flamelicks, rendererClasses [&NiagaraMeshRendererProperties]}
  {emitterName=Particle_Spark_Sphere001, rendererClasses [&NiagaraSpriteRendererProperties]}
  {emitterName=Light001, rendererClasses [&NiagaraLightRendererProperties]}
  {emitterName=Magic_Sparklies_Burst, rendererClasses [&NiagaraSpriteRendererProperties]}
  {emitterName=Chromatic_Flare, rendererClasses [&NiagaraSpriteRendererProperties]}
  {emitterName=Black_Back, rendererClasses [&NiagaraSpriteRendererProperties]}
 ]
 userVariables []
}
